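# Material function for the FineArt DecaStatue prop, building its crystal look with base color, glowing interior cracks and destructible damage.
# Bump-offset UVs fake crack depth, while emissive layers combine an internal sphere-mask glow with platform-switched intensities and eye-adaptation compensation.
# Fake Phong specular, virtual-plane interior reflection, screen-space glitter and skylight sampling supply faked shine.
# It outputs material attributes plus intermediate masks, specular, color, emissive and offset UVs.
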MaterialFunction /Hera/Props/FineArt/FineArt_DecaStatue_A/Materials/MF_FIneArtBase
graph {nodes [
  n1 = NamedRerouteUsage()
  n2 = ScalarParameter() {defaultValue=1, group=Interior Cracks, parameterName=Emissive Cracks}
  n3 = SphereMask(
   A=LocalPosition().XYZ
   B=MaterialFunctionCall(R=Constant(), G=Constant(), B=ScalarParameter() {defaultValue=205.2454, group=Internal Glow, parameterName=Sphere Mask Height}) {materialFunction=&AppendMany}.RGB
   Radius=ScalarParameter() {defaultValue=373.4858, group=Internal Glow, parameterName=Sphere Mask Radius}
   Hardness=Constant() {r=-2.3889}
  )
  n4 = Multiply(
   A=Add(A=Multiply(A=@n1, B=@n2) {constB=0.5}, B=NamedRerouteUsage())
   B=Multiply(A=VectorParameter() {group=Interior Cracks, parameterName=Aditional Glow color, defaultValue=<rgba 1 0 0.9598 1>}.RGB, B=@n3)
  )
  n5 = Saturate(None=Add(A=LocalPosition().Z) {constB=-150})
  n6 = ComponentMask(None=VertexNormalWS()) {b=true}
  Add(A=Multiply(A=@n4, B=Multiply(A=@n5, B=@n6)), B=NamedRerouteUsage())
  Constant()
  Constant()
  n7 = NamedRerouteUsage()
  n8 = ScalarParameter() {defaultValue=3, group=Interior Cracks, parameterName=Emissive Intensity}
  n9 = ScalarParameter() {defaultValue=7, group=Interior Cracks, parameterName=Low Emissive intensity}
  n10 = Multiply(
   A=MaterialFunctionCall("Xb1,PS4"=@n8, "PC-High&&Epic,XSX,PS5"=@n8, "PC-Perf,Android"=@n9, Switch=@n8, "PC-Low(NotPerf)&&Medium"=@n9) {materialFunction=&PlatformSwitch_1D}.Result
   B=@n4
  )
  n11 = NamedRerouteUsage()
  n12 = Add(A=NamedRerouteUsage(), B=Add(A=NamedRerouteUsage(), B=Add(A=@n10, B=@n11)))
  n13 = SetMaterialAttributes!(
   MaterialAttributes
   BaseColor=NamedRerouteUsage()
   Metallic
   Specular=NamedRerouteUsage()
   Roughness=NamedRerouteUsage()
   EmissiveColor=EyeAdaptationInverse(LightValueInput=Add(A=@n7, B=@n12), AlphaInput=Constant() {r=0.3}).EyeAdaptationInverse
   Opacity
   Normal=NamedRerouteUsage()
   SubsurfaceColor
   WorldPositionOffset
  ) {attributeSetTypes [
    #1
    #2
    #3
    #4
    #5
    #6
    #7
    #8
    #9
   ]}
  FunctionOutput(None=MaterialFunctionCall(Material_In=@n13) {materialFunction=&MF_FN_DestructibleDamage}.Material_Out) {outputName=MA}
  FunctionOutput(None=NamedRerouteDeclaration(None=@n3) {name=Sphere Mask, nodeColor=<rgba 0 1 0.2266 1>}) {outputName=Sphere Mask, sortPriority=1}
  FunctionOutput(None=NamedRerouteUsage()) {outputName=Crystal Mask, sortPriority=1}
  n14 = NamedRerouteUsage()
  n15 = CollectionParameter() {
   collection=&FortniteMaterialParameters
   parameterName=SunlightDirectionalVector
  }
  n16 = ComponentMask(None=@n15) {b=true, g=true, r=true}
  n17 = PixelNormalWS()
  n18 = DotProduct(A=@n16, B=@n17)
  n19 = Subtract(A=Multiply(A=Multiply(A=@n18) {constB=2}, B=@n17), B=@n16)
  n20 = MaterialFunctionCall() {materialFunction=&CameraDirectionVector}
  n21 = Max(A=DotProduct(A=Normalize(VectorInput=@n19), B=Multiply(A=@n20.Result) {constB=-1})) {constB=0}
  n22 = ScalarParameter() {defaultValue=275.1841, group=Fake Phong, parameterName=Fake Phong Contrast}
  n23 = Multiply(
   A=Multiply(A=@n14, B=NamedRerouteUsage())
   B=Multiply(A=Power(Base=@n21, Exp=@n22), B=ScalarParameter() {defaultValue=22.9038, group=Fake Phong, parameterName=Fake Phong Intensity})
  )
  FunctionOutput(None=NamedRerouteDeclaration(None=Saturate(None=@n23)) {name=Fake Spec, nodeColor=<rgba 0.7578 0 1 1>}) {outputName=Fake Specular, sortPriority=1}
  FunctionOutput(None=NamedRerouteUsage()) {outputName=Base Color + Cracks, sortPriority=1}
  n24 = VectorParameter() {parameterName=Color Multiply, defaultValue=<rgba 0.8073 0.8073 0.8073 1>}
  n25 = Multiply(A=@n24.RGB) {constB=3}
  n26 = Multiply(
   A=MaterialFunctionCall("Xb1,PS4"=@n24.RGB, "PC-High&&Epic,XSX,PS5"=@n24.RGB, "PC-Perf,Android"=@n25, Switch=@n24.RGB, "PC-Low(NotPerf)&&Medium"=@n25) {materialFunction=&PlatformSwitch_1D}.Result
   B=TextureSampleParameter2D() {parameterName=Base Color, texture=/Hera/Props/FineArt/FineArt_DecaStatue_A/Textures/T_FineArt_DecaStatue_A_A_D}.RGB
  )
  n27 = TextureSampleParameter2D() {parameterName=Mask, samplerType=SAMPLERTYPE_Masks, texture=/Hera/Props/FineArt/FineArt_DecaStatue_A/Textures/T_FineArt_DecaStatue_A_A_M}
  n28 = LinearInterpolate(
   B=Multiply(A=Multiply(A=@n26, B=@n27.G)) {constB=2}
   Alpha=Saturate(None=ScalarParameter() {defaultValue=1, parameterName=Use Emissive Mask?})
  ) {constA=1}
  FunctionOutput(None=NamedRerouteDeclaration(None=Saturate(None=@n28)) {name=Emissive, nodeColor=<rgba 0.9688 1 0 1>}) {outputName=Base Emissive, sortPriority=1}
  LinearInterpolate(A=@n14, B=NamedRerouteUsage())
  LinearInterpolate()
  n29 = ScreenPosition()
  n30 = TextureCoordinate()
  n31 = TextureSample(UVs=Multiply(A=Multiply(A=@n29.ViewportUV, B=@n30), B=ScalarParameter() {defaultValue=98.6753, group=Glitter, parameterName=Glittter Size}) {constB=3}) {texture=/Game/Athena/SafeZone/PulseLancerAssets/T_StarSky_SafeZone}
  n32 = Multiply(A=@n31.R, B=@n31.R)
  Multiply(A=@n32, B=@n32)
  n33 = SkyLightEnvMapSample(Direction=ReflectionVectorWS(), Roughness=ScalarParameter() {defaultValue=0.75, group=EnvSkySample, parameterName=SkySampleRoughness})
  n34 = ScalarParameter() {defaultValue=3, group=EnvSkySample, parameterName=SkySampleContrast}
  n35 = Multiply(
   A=NamedRerouteUsage()
   B=Multiply(A=Power(Base=@n33, Exp=@n34), B=ScalarParameter() {defaultValue=3, group=EnvSkySample, parameterName=SkySampleBrightness})
  )
  NamedRerouteDeclaration(None=@n35) {name=Cubemap, nodeColor=<rgba 1 0 0.5156 1>}
  n36 = Fresnel(ExponentIn=ScalarParameter() {defaultValue=4, group=Emissive, parameterName=GlowFresnel}, Normal=PixelNormalWS())
  NamedRerouteDeclaration(None=OneMinus(None=NamedRerouteDeclaration(None=@n36) {name=Fresnel, nodeColor=<rgba 0.9922 1 0 1>})) {name=Fresnel_Inverted, nodeColor=<rgba 0.9922 1 0 1>}
  NamedRerouteDeclaration(None=Constant() {r=0.1}) {name=Rough, nodeColor=<rgba 0 1 0.3906 1>}
  NamedRerouteDeclaration(None=Constant() {r=0.15}) {name=Spec, nodeColor=<rgba 1 0.0703 0 1>}
  NamedRerouteDeclaration(None=@n27.B) {name=Crystal Mask, nodeColor=<rgba 1 0 0 1>}
  n37 = ScalarParameter() {defaultValue=-0.0816, group=InteriorCracks, parameterName=InteriorCracks_BumpOffset}
  n38 = CameraVectorWS()
  n39 = ComponentMask(None=Transform(None=@n38) {transformSourceType=TRANSFORMSOURCE_World, transformType=TRANSFORM_Tangent}) {g=true, r=true}
  n40 = Transform(None=Constant3Vector() {constant=<rgba 0 0 1 0>})
  n41 = Add(
   A=Multiply(A=TextureCoordinate(), B=ScalarParameter() {defaultValue=1, group=InteriorCracks, parameterName=InteriorCracks_Tiling, sortPriority=1})
   B=Divide(A=Multiply(A=@n37, B=@n39), B=DotProduct(A=@n38, B=@n40))
  )
  n42 = TextureSample(UVs=@n41) {texture=&/Crystal_M}
  NamedRerouteDeclaration(None=LinearInterpolate(A=Constant(), B=Constant() {r=0.112}, Alpha=FunctionOutput(None=@n42.R) {outputName=Cracks Texture Mask, sortPriority=1})) {name=Crack Mask, nodeColor=<rgba 0.2656 1 0 1>}
  NamedRerouteDeclaration(None=FunctionOutput(None=@n41) {outputName=Offseted UVs, sortPriority=1}) {name=Offset Uvs, nodeColor=<rgba 0 0.6953 1 1>}
  n43 = ScalarParameter() {defaultValue=9.5, group=Fake Reflection, parameterName=Fake Reflection Intensity}
  n44 = Multiply(A=@n43) {constB=0.4}
  n45 = Multiply(A=NamedRerouteUsage(), B=LinearInterpolate(A=NamedRerouteUsage()))
  n46 = Multiply(A=VertexNormalWS())
  n47 = Constant3Vector() {constant=<rgba 0 0 1 0>}
  n48 = ScalarParameter() {defaultValue=200, parameterName=Fake Reflectio Position}
  MaterialExpressionMaterialFunctionCall_18 = MaterialFunctionCall(
   Plane Normal Axis=CameraVectorWS()
   Plane X Axis=Normalize(VectorInput=Constant3Vector() {constant=<rgba 0 0 1 0>})
   Plane Center=Add(A=ObjectPositionWS(), B=Multiply(A=@n47, B=@n48))
   UV Scale=ScalarParameter() {defaultValue=527.3169, group=Fake Reflection, parameterName=Fake Reflection UV Scale}
  ) {materialFunction=&VirtualPlaneCoordinates}
  n49 = Convert(None=@MaterialExpressionMaterialFunctionCall_18.UVs) {
   nodeName=Break Float2
   convertInputs [{
     type=Vector2
     expressionInput {
      expression=&self:MaterialExpressionMaterialFunctionCall_18
      inputName=None
      mask=0
      maskA=0
      maskB=0
      maskG=0
      maskR=0
      outputIndex=0
     }
    }]
   convertOutputs [{type=Scalar}, {type=Scalar}]
  }
  n50 = LinearInterpolate(A=Multiply(A=TextureCoordinate()) {constB=5}, B=AppendVector(A=@n49, B=Multiply(A=@n49) {constB=-1})) {constAlpha=0.9}
  n51 = Constant2Vector() {g=2, r=1}
  n52 = TextureSample!(UVs=Add(A=ComponentMask(None=@n46) {g=true, r=true}, B=Multiply(A=@n50, B=@n51)), Tex, Level=Constant() {r=3}, Apply View MipBias) {mipValueMode=TMVM_MipLevel, texture=/Game/Environments/Asteria/Sets/NeonCity/Textures/T_FakeInterior03_RT}
  NamedRerouteDeclaration(None=Multiply(
    A=MaterialFunctionCall("Xb1,PS4"=@n43, "PC-High&&Epic,XSX,PS5"=@n43, "PC-Perf,Android"=@n44, Switch=@n43, "PC-Low(NotPerf)&&Medium"=@n44) {materialFunction=&PlatformSwitch_1D}.Result
    B=Multiply(A=@n45, B=@n52.RGB)
   )) {name=Fake Interior, nodeColor=<rgba 0 1 0.3203 1>}
  NamedRerouteDeclaration(None=MaterialFunctionCall() {materialFunction=&Color_Time_of_Day}.Result) {name=Color of day, nodeColor=<rgba 0 0.7656 1 1>}
  n53 = TextureSample!(UVs=Multiply(A=NamedRerouteUsage(), B=ScalarParameter() {defaultValue=9.5409, parameterName=Glitter size 2}), Tex, Apply View MipBias) {texture=&T_StarSky_SafeZone}
  n54 = Multiply(A=@n53.R, B=@n53.R)
  n55 = Constant() {r=4}
  n56 = Multiply(A=NamedRerouteUsage(), B=Multiply(A=Power(Base=@n32, Exp=@n55)) {constB=6})
  NamedRerouteDeclaration(None=Multiply(A=Multiply(A=@n54, B=@n56) {constA=1}, B=ScalarParameter() {defaultValue=460.8319, group=Glitter, parameterName=Glitter int})) {name=Glitter, nodeColor=<rgba 1 0 0.2812 1>}
  n57 = NamedRerouteUsage()
  n58 = LinearInterpolate(
   A=ScalarParameter() {defaultValue=0.746, group=InteriorCracks, parameterName=InteriorContrast_low}
   B=ScalarParameter() {defaultValue=2, group=InteriorCracks, parameterName=InteriorContrast_high}
   Alpha=@n42.R
  )
  NamedRerouteDeclaration(None=Multiply(A=Multiply(A=@n57, B=@n58), B=NamedRerouteUsage())) {name=InteriorCracks, nodeColor=<rgba 0 0.2734 1 1>}
  NamedRerouteDeclaration(None=@n26) {name=Base color, nodeColor=<rgba 0 0.6484 1 1>}
  NamedRerouteDeclaration(None=Constant3Vector() {constant=<rgba 0 0 1 0>}) {name=Normal, nodeColor=<rgba 0.7109 0 1 1>}
  NamedRerouteUsage()
  OneMinus(None=@n19)
  n59 = NamedRerouteUsage()
  Reroute(None=Multiply(A=Add(A=@n35, B=@n59), B=ScalarParameter() {defaultValue=0.05, parameterName=SSSBrightnessMaster}))
  n60 = ScalarParameter() {defaultValue=0.3859, group=Emissive, parameterName=GlowAttenuation}
  n61 = Divide(
   A=ObjectRadius()
   B=MaterialFunctionCall() {materialFunction=&ObjectScale}.Scale Z
  )
  n62 = ScalarParameter() {defaultValue=-4.8763, group=Emissive, parameterName=DistortGlow_fromCrackNormals}
  n63 = MaterialFunctionCall(Float3=NamedRerouteUsage()) {materialFunction=&BreakOutFloat3Components}
  n64 = Subtract(A=@n63.R, B=@n63.R)
  n65 = Divide(A=Length(None=LocalPosition().XYZ) {is=class MaterialExpressionLength}, B=Add(A=Multiply(A=@n60, B=@n61), B=Multiply(A=@n62, B=@n64)))
  n66 = Divide(A=Constant() {r=1}, B=Multiply(A=@n65, B=@n65))
  n67 = ScalarParameter() {defaultValue=1, group=Emissive, parameterName=MaxGlow}
  n68 = Multiply(
   A=VectorParameter() {group=Emissive, parameterName=EmissiveGlow, defaultValue=<rgba 1.2986 2.63 5 1>}.RGB
   B=Multiply(A=NamedRerouteUsage(), B=Min(A=@n66, B=@n67))
  )
  n69 = Add(A=NamedRerouteUsage(), B=Multiply(A=NamedRerouteUsage(), B=Add(A=@n68, B=@n35)))
  n70 = ScalarParameter() {defaultValue=0.376, parameterName=BrightnessMaster}
  Reroute(None=NamedRerouteDeclaration(None=Multiply(A=@n69, B=@n70)) {name=Base cracks, nodeColor=<rgba 1 0.1406 0 1>})
  VectorParameter() {group=Shading, parameterName=CrackColor, sortPriority=2, defaultValue=<rgba 0.2217 0.6426 1 1>}
 ]}
