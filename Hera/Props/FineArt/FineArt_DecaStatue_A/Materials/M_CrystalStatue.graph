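# Default Lit crystal statue material for the DecaStatue A fine-art prop, built on the shared MF_FIneArtBase function using material attributes.
# The function's attributes pass through a set node that overrides Specular to one and Roughness to zero for a glossy finish.
# Named reroutes expose the function's crystal, crack and sphere masks, fake specular, cracked color and emissive.
# A loose multiply/lerp and Param_1 appear unconnected to the result.
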Material /Hera/Props/FineArt/FineArt_DecaStatue_A/Materials/M_CrystalStatue
graph {
 properties {bAutomaticallySetUsageInEditor=false, bUseMaterialAttributes=true, usedShadingModels=Default Lit}
 nodes [
  Multiply(A=LinearInterpolate(Alpha=NamedRerouteUsage()) {constA=1})
  n1 = MaterialFunctionCall() {materialFunction=&/MF_FIneArtBase}
  NamedRerouteDeclaration(None=@n1.Crystal Mask) {name=Crystal Mask, nodeColor=<rgba 1 0 0 1>}
  NamedRerouteDeclaration(None=@n1.Cracks Texture Mask) {name=Crack Mask, nodeColor=<rgba 0.2656 1 0 1>}
  NamedRerouteDeclaration(None=@n1.Fake Specular) {name=Fake Spec, nodeColor=<rgba 0.7578 0 1 1>}
  NamedRerouteDeclaration(None=@n1.Sphere Mask) {name=Sphere Mask, nodeColor=<rgba 0 1 0.2266 1>}
  NamedRerouteDeclaration(None=@n1.Base Color + Cracks) {name=Color + Cracks, nodeColor=<rgba 0 0.2734 1 1>}
  NamedRerouteDeclaration(None=@n1.Base Emissive) {name=Emissive, nodeColor=<rgba 0.9688 1 0 1>}
  ScalarParameter() {parameterName=Param_1}
  n2 = GetMaterialAttributes(None=@n1.MA) {attributeGetTypes [
    #1
    #2
    #3
    #4
    #5
    #6
    #7
    #8
   ]}
  SetMaterialAttributes!(
   MaterialAttributes=@n2.MaterialAttributes
   Base Color
   Metallic
   Specular=Constant() {r=1}
   Normal
   Tangent
   Emissive Color
   Opacity
   Roughness=Constant()
  ) {attributeSetTypes [
    #1
    #2
    #3
    #5
    #9
    #6
    #7
    #4
   ]}
 ]
}
