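# Single Layer Water material for the DecaStatue crystal statue prop, giving it a translucent, glowing, cracked crystal look.
# Interior cracks use view-offset parallax sampling, while world-aligned crack normals add surface detail.
# Emissive builds from inverse-square core glow, sky-light reflection samples, time-of-day fake sun specular and a sphere-masked, sine-pulsed color glow, with eye-adaptation compensation.
# Texture maps supply surface attributes and masks; scatter and absorption parameters feed the water output.
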Material /Hera/Props/FineArt/FineArt_DecaStatue_A/Materials/M_CrystalStatueSingleLayer
graph {
 properties {bAutomaticallySetUsageInEditor=false, bUseMaterialAttributes=true, shadingModel=MSM_SingleLayerWater, usedShadingModels=SingleLayerWater}
 nodes [
  n1 = Fresnel(ExponentIn=ScalarParameter() {defaultValue=0.3747, group=Emissive, parameterName=GlowFresnel}, Normal=PixelNormalWS())
  NamedRerouteDeclaration(None=OneMinus(None=NamedRerouteDeclaration(None=@n1) {name=Fresnel, nodeColor=<rgba 0.9922 1 0 1>})) {name=Fresnel_Inverted, nodeColor=<rgba 0.9922 1 0 1>}
  n2 = TextureSample() {samplerType=SAMPLERTYPE_Masks, texture=/Hera/Props/FineArt/FineArt_DecaStatue_A/Textures/T_FineArt_DecaStatue_A_A_S}
  NamedRerouteDeclaration(None=@n2.R) {name=Rough, nodeColor=<rgba 0 1 0.3906 1>}
  NamedRerouteDeclaration(None=@n2.B) {name=Spec, nodeColor=<rgba 1 0.0703 0 1>}
  n3 = TextureSample() {samplerType=SAMPLERTYPE_Masks, texture=/Hera/Props/FineArt/FineArt_DecaStatue_A/Textures/T_FineArt_DecaStatue_A_A_M}
  NamedRerouteDeclaration(None=@n3.B) {name=Crystal Mask, nodeColor=<rgba 1 0 0 1>}
  n4 = TextureCoordinate()
  n5 = ScalarParameter() {defaultValue=0.7116, group=InteriorCracks, parameterName=InteriorCracks_Tiling, sortPriority=1}
  n6 = CameraVectorWS()
  n7 = Multiply(
   A=ScalarParameter() {defaultValue=-0.0616, group=InteriorCracks, parameterName=InteriorCracks_BumpOffset}
   B=ComponentMask(None=Transform(None=@n6) {transformSourceType=TRANSFORMSOURCE_World, transformType=TRANSFORM_Tangent}) {g=true, r=true}
  )
  n8 = NamedRerouteUsage()
  n9 = ScalarParameter() {defaultValue=-1, group=InteriorCracks, parameterName=InteriorDistortion Normal}
  n10 = DotProduct(
   A=@n6
   B=Transform(None=MaterialFunctionCall(Normal=@n8, Flatness=@n9) {materialFunction=&FlattenNormal}.Result)
  )
  n11 = TextureSample(UVs=Add(A=Multiply(A=@n4, B=@n5), B=Divide(A=@n7, B=@n10))) {texture=&/Crystal_M}
  NamedRerouteDeclaration(None=@n11.R) {name=Crack Mask, nodeColor=<rgba 0.2656 1 0 1>}
  n12 = MaterialFunctionCall() {materialFunction=&Color_Time_of_Day}
  n13 = Add(A=SkyAtmosphereLightDirection(), B=CameraVectorWS())
  n14 = Power(
   Base=DotProduct(A=Normalize(VectorInput=@n13), B=PixelNormalWS())
   Exp=ScalarParameter() {defaultValue=14, parameterName=Fake Specular Contrast}
  )
  NamedRerouteDeclaration(None=Saturate(None=Multiply(A=@n12.Result, B=@n14))) {name=Fake Spec, nodeColor=<rgba 0.7578 0 1 1>}
  n15 = MaterialFunctionCall(
   TextureObject=TextureObject() {samplerType=SAMPLERTYPE_Normal, texture=&/Crystal_N}
   TextureSize=ScalarParameter() {defaultValue=1312.7904, group=Cracks, parameterName=CrackNormals_Size}
   WorldSpace=StaticBool()
   ProjectionTransitionContrast=ScalarParameter() {defaultValue=50, group=Cracks, parameterName=Cracks_ProjectionContrast}
  ) {materialFunction=&WorldAlignedNormal}
  NamedRerouteDeclaration(None=MaterialFunctionCall(
    Normal=NamedRerouteDeclaration(None=@n15.XYZ Texture) {name=NormalMapTriplanar, nodeColor=<rgba 0 1 0.3672 1>}
    Flatness=ScalarParameter() {defaultValue=0.8, group=Cracks, parameterName=Cracks_Normal_Flatten}
   ) {materialFunction=&FlattenNormal}.Result) {name=NormalMap_Output, nodeColor=<rgba 0 1 0.3672 1>}
  n16 = NamedRerouteUsage()
  n17 = LinearInterpolate(
   A=ScalarParameter() {defaultValue=0.7665, group=InteriorCracks, parameterName=InteriorContrast_low}
   B=ScalarParameter() {defaultValue=4.6517, group=InteriorCracks, parameterName=InteriorContrast_high}
   Alpha=@n11.R
  )
  NamedRerouteDeclaration(None=Multiply(A=Multiply(A=@n16, B=@n17), B=NamedRerouteUsage())) {name=InteriorCracks, nodeColor=<rgba 0 0.2734 1 1>}
  n18 = Distance(A=Constant3Vector() {desc=WorldSpace Location, constant=<rgba 1708.7944 604.3872 69.4823 0>}, B=WorldPosition().XYZ)
  n19 = Power(
   Base=Divide(A=@n18, B=ScalarParameter() {defaultValue=860.6788, group=Emissive, parameterName=CoreDarken_Distance})
   Exp=ScalarParameter() {defaultValue=3, group=Emissive, parameterName=CoreDarken_Power}
  )
  NamedRerouteDeclaration(None=Max(A=Saturate(None=@n19), B=ScalarParameter() {defaultValue=0.05, group=Emissive, parameterName=CoreDarken_Min})) {name=CoreDarkenMask, nodeColor=<rgba 1 0 0.0469 1>}
  n20 = TextureSample() {texture=/Hera/Props/FineArt/FineArt_DecaStatue_A/Textures/T_FineArt_DecaStatue_A_A_D}
  NamedRerouteDeclaration(None=@n20.RGB) {name=Base color, nodeColor=<rgba 0 0.6484 1 1>}
  n21 = Multiply(A=@n20.RGB, B=@n3.G)
  NamedRerouteDeclaration(None=Saturate(None=Multiply(A=@n21) {constB=2})) {name=Emissive, nodeColor=<rgba 0.9688 1 0 1>}
  NamedRerouteDeclaration(None=TextureSample() {samplerType=SAMPLERTYPE_Normal, texture=/Hera/Props/FineArt/FineArt_DecaStatue_A/Textures/T_FineArt_DecaStatue_A_A_N}.RGB) {name=Normal, nodeColor=<rgba 0.7109 0 1 1>}
  NamedRerouteUsage()
  n22 = SkyLightEnvMapSample(Direction=ReflectionVectorWS(), Roughness=ScalarParameter() {defaultValue=0.75, group=EnvSkySample, parameterName=SkySampleRoughness})
  n23 = ScalarParameter() {defaultValue=3, group=EnvSkySample, parameterName=SkySampleContrast}
  n24 = Multiply(
   A=NamedRerouteUsage()
   B=Multiply(A=Power(Base=@n22, Exp=@n23), B=ScalarParameter() {defaultValue=3, group=EnvSkySample, parameterName=SkySampleBrightness})
  )
  n25 = NamedRerouteUsage()
  Reroute(None=Multiply(A=Add(A=@n24, B=@n25), B=ScalarParameter() {defaultValue=0.05, parameterName=SSSBrightnessMaster}))
  n26 = LocalPosition()
  n27 = Multiply(
   A=ScalarParameter() {defaultValue=0.3859, group=Emissive, parameterName=GlowAttenuation}
   B=Divide(
    A=ObjectRadius()
    B=MaterialFunctionCall() {materialFunction=&ObjectScale}.Scale Z
   )
  )
  n28 = MaterialFunctionCall(Float3=NamedRerouteUsage()) {materialFunction=&BreakOutFloat3Components}
  n29 = Multiply(A=ScalarParameter() {defaultValue=-4.8763, group=Emissive, parameterName=DistortGlow_fromCrackNormals}, B=Subtract(A=@n28.R, B=@n28.R))
  n30 = StaticSwitchParameter(
   True=Divide(A=Length(None=@n26.XYZ) {is=class MaterialExpressionLength}, B=Add(A=@n27, B=@n29))
   False=Divide(A=@n18, B=ScalarParameter() {defaultValue=860.6788, group=Emissive, parameterName=GlowAttenuationWS})
  ) {defaultValue=true, group=Emissive, parameterName=GlowDistanceLocal}
  n31 = Divide(A=Constant() {r=1}, B=Multiply(A=@n30, B=@n30))
  n32 = ScalarParameter() {defaultValue=1, group=Emissive, parameterName=MaxGlow}
  n33 = Multiply(
   A=VectorParameter() {group=Emissive, parameterName=EmissiveGlow, defaultValue=<rgba 1.2986 2.63 5 1>}.RGB
   B=Multiply(A=NamedRerouteUsage(), B=Min(A=@n31, B=@n32))
  )
  n34 = Add(A=NamedRerouteUsage(), B=Multiply(A=NamedRerouteUsage(), B=Add(A=@n33, B=@n24)))
  n35 = ScalarParameter() {defaultValue=0.376, parameterName=BrightnessMaster}
  Reroute(None=NamedRerouteDeclaration(None=Multiply(A=@n34, B=@n35)) {name=Base cracks, nodeColor=<rgba 1 0.1406 0 1>})
  n36 = NamedRerouteUsage()
  n37 = LocalPosition()
  n38 = MaterialFunctionCall(R=Constant(), G=Constant(), B=Constant() {r=205.2454}) {materialFunction=&AppendMany}
  n39 = Constant() {r=373.4858}
  n40 = Constant() {r=-2.3889}
  n41 = Multiply(
   A=Add(A=Multiply(A=@n36) {constB=0.5}, B=NamedRerouteUsage())
   B=Multiply(
    A=VectorParameter() {parameterName=Aditional Glow color, defaultValue=<rgba 1 0 0.9598 1>}.RGB
    B=SphereMask(A=@n37.XYZ, B=@n38.RGB, Radius=@n39, Hardness=@n40)
   )
  )
  n42 = Add(A=LocalPosition().Z) {constB=-150}
  n43 = VertexNormalWS()
  n44 = Multiply(A=@n41, B=Multiply(A=Saturate(None=@n42), B=ComponentMask(None=@n43) {b=true}))
  n45 = NamedRerouteUsage()
  n46 = Time()
  n47 = Add(A=Sine(None=Multiply(A=@n46) {constB=0.3}))
  n48 = Add(A=LinearInterpolate(Alpha=Multiply(A=@n47) {constB=0.5}) {constB=2}, B=ScalarParameter() {defaultValue=4, parameterName=Emissive Intensity})
  n49 = Add(A=NamedRerouteUsage(), B=Add(A=Multiply(A=@n48, B=@n41), B=NamedRerouteUsage()))
  n50 = Constant() {r=0.3}
  n51 = NamedRerouteUsage()
  n52 = ScalarParameter() {defaultValue=0.7491, group=Shading, parameterName=SSSOpacity}
  SetMaterialAttributes!(
   MaterialAttributes
   Base Color=NamedRerouteUsage()
   Metallic
   Specular=NamedRerouteUsage()
   Roughness=NamedRerouteUsage()
   Emissive Color=LinearInterpolate(A=Add(A=@n44, B=@n45), B=EyeAdaptationInverse(LightValueInput=@n49, AlphaInput=@n50).EyeAdaptationInverse, Alpha=NamedRerouteUsage())
   Opacity=LinearInterpolate(B=Add(A=@n51, B=@n52), Alpha=NamedRerouteUsage()) {constA=1}
   Normal=NamedRerouteUsage()
   Subsurface Color
  ) {attributeSetTypes [
    #1
    #2
    #3
    #4
    #5
    #6
    #7
    #8
   ]}
  n53 = VectorParameter() {group=SingleLayerShading, parameterName=Scatter, defaultValue=<rgba 0.0855 0.2775 0.9822 0.1>}
  n54 = Saturate(None=Constant() {r=1})
  n55 = LinearInterpolate(A=@n53.A, B=ScalarParameter() {defaultValue=-0.3563, group=SingleLayerShading, parameterName=EdgeScatter}, Alpha=@n54)
  n56 = Constant() {r=1}
  n57 = VectorParameter() {group=SingleLayerShading, parameterName=Absorption, defaultValue=<rgba 0.125 0.125 0.125 29.1745>}
  n58 = Multiply(
   A=@n57.RGB
   B=LinearInterpolate(A=@n57.A, B=ScalarParameter() {defaultValue=1269.7767, group=SingleLayerShading, parameterName=EdgeAbsorption}, Alpha=@n54)
  )
  n59 = NamedRerouteUsage()
  n60 = PixelDepth()
  SingleLayerWaterMaterialOutput!(
   ScatteringCoefficients=Multiply(A=@n53.RGB, B=Divide(A=@n55) {constB=100})
   AbsorptionCoefficients=Divide(A=Divide(A=@n56, B=@n58), B=Subtract(A=@n59, B=@n60.R))
   PhaseG
   ColorScaleBehindWater
  )
  VectorParameter() {group=Shading, parameterName=CrackColor, sortPriority=2, defaultValue=<rgba 0.2217 0.6426 1 1>}
  VertexNormalWS()
 ]
}
