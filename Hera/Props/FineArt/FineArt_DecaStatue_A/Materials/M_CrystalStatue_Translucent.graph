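# Translucent lit material for the FineArt crystal Deca statue, built on MF_FIneArtBase with zero roughness and full specular.
# Opacity sums surface coverage with two weighted mask terms, apparently cracks and sphere mask.
# Two squared star-sky texture samples, one using screen-blended UVs, form emissive color blobs tinted partly by time of day.
# Quality switches drop blobs and a base-color multiplier on Low and Medium; several time-of-day and depth opacity nodes are unconnected.
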Material /Hera/Props/FineArt/FineArt_DecaStatue_A/Materials/M_CrystalStatue_Translucent
graph {
 properties {
  bAutomaticallySetUsageInEditor=false
  bUseMaterialAttributes=true
  blendMode=BLEND_Translucent
  translucencyLightingMode=TLM_VolumetricPerVertexDirectional
  usedShadingModels=Default Lit
 }
 nodes [
  n1 = ScalarParameter() {defaultValue=0.829, group=Opacity, parameterName=Surfacecoverage}
  n2 = CollectionParameter() {
   collection=&FortniteMaterialParameters
   parameterName=FortniteActiveTimeOfDay
  }
  n3 = MaterialFunctionCall(Float4=Saturate(None=@n2)) {materialFunction=&BreakOutFloat4Components}
  n4 = Saturate(None=Add(A=@n3.R, B=@n3.R))
  n5 = NamedRerouteDeclaration(None=@n3.A) {name=NightTime, nodeColor=<rgba 0 1 0.6016 1>}
  Add(A=@n1, B=Add(A=LinearInterpolate(Alpha=@n4) {constB=0.075}, B=LinearInterpolate(Alpha=@n5) {constB=0.06}))
  Constant() {r=1}
  Constant()
  LinearInterpolate(B=Add(A=NamedRerouteUsage(), B=ScalarParameter() {defaultValue=1, group=Shading, parameterName=SSSOpacity}), Alpha=NamedRerouteUsage()) {constA=1}
  n6 = NamedRerouteUsage()
  n7 = ScalarParameter() {defaultValue=-0.0594, group=Opacity, parameterName=Cracks opacity}
  n8 = NamedRerouteUsage()
  n9 = ScalarParameter() {defaultValue=0.85, group=Opacity, parameterName=Sphere Mask Opacity}
  n10 = Add(A=Add(A=Multiply(A=@n6, B=@n7) {constB=0.2}, B=Multiply(A=@n8, B=@n9)) {constB=0.1}, B=@n1)
  LinearInterpolate(B=@n10, Alpha=NamedRerouteUsage()) {constA=1}
  n11 = PixelDepth()
  n12 = Constant() {r=3000}
  LinearInterpolate(A=@n1, Alpha=Saturate(None=Divide(A=@n11.R, B=@n12)))
  n13 = NamedRerouteUsage()
  Multiply(A=ScreenPosition().ViewportUV, B=@n13)
  n14 = MaterialFunctionCall() {materialFunction=&/MF_FIneArtBase}
  NamedRerouteDeclaration(None=@n14.Base Emissive) {name=Emissive, nodeColor=<rgba 0.9688 1 0 1>}
  NamedRerouteDeclaration(None=@n14.Crystal Mask) {name=Crystal Mask, nodeColor=<rgba 1 0 0 1>}
  NamedRerouteDeclaration(None=@n14.Cracks Texture Mask) {name=Crack Mask, nodeColor=<rgba 0.2656 1 0 1>}
  NamedRerouteDeclaration(None=@n14.Fake Specular) {name=Fake Spec, nodeColor=<rgba 0.7578 0 1 1>}
  NamedRerouteDeclaration(None=@n14.Sphere Mask) {name=Sphere Mask, nodeColor=<rgba 0 1 0.2266 1>}
  NamedRerouteDeclaration(None=@n10) {name=Opacity, nodeColor=<rgba 0.1484 0 1 1>}
  NamedRerouteDeclaration(None=@n14.Offseted UVs) {name=OFfset UVS, nodeColor=<rgba 0.6875 0 1 1>}
  n15 = LinearInterpolate(
   A=VectorParameter() {group=Color Blobs, parameterName=Color Blobs color, defaultValue=<rgba 0.0747 0.2812 0.3333 1>}.RGB
   B=MaterialFunctionCall() {materialFunction=&Color_Time_of_Day}.Result
  )
  n16 = NamedRerouteUsage()
  n17 = TextureSample(UVs=Multiply(A=@n13, B=ScalarParameter() {defaultValue=1.2877, group=Color Blobs, parameterName=Color Blobs Scale 2})) {texture=/Game/Athena/SafeZone/PulseLancerAssets/T_StarSky_SafeZone}
  n18 = NamedRerouteUsage()
  n19 = LinearInterpolate(A=ScreenPosition().ViewportUV, B=@n18)
  n20 = TextureSample(UVs=Multiply(A=Multiply(A=@n19, B=@n18), B=ScalarParameter() {defaultValue=0.432, group=Color Blobs, parameterName=Color Blobs Scale}) {constB=3}) {texture=&T_StarSky_SafeZone}
  n21 = Multiply(A=Multiply(A=@n17.R, B=@n17.R), B=Multiply(A=@n20.R, B=@n20.R)) {constA=1}
  n22 = Multiply(
   A=Multiply(A=Saturate(None=@n15), B=Multiply(A=@n16, B=@n21))
   B=ScalarParameter() {defaultValue=16.0016, group=Color Blobs, parameterName=Color Blobs Intensity}
  )
  n23 = Constant()
  NamedRerouteDeclaration(None=QualitySwitch(Default=@n22, Low=@n23, High=@n22, Medium=@n23, Epic=@n22)) {name=Blobs, nodeColor=<rgba 1 0 0.2812 1>}
  NamedRerouteDeclaration(None=@n14.Base Color + Cracks) {name=Color + Cracks, nodeColor=<rgba 0 0.2734 1 1>}
  n24 = GetMaterialAttributes(None=@n14.MA) {attributeGetTypes [
    #1
    #2
    #3
    #4
    #5
    #6
    #7
    #8
   ]}
  n25 = NamedRerouteUsage()
  n26 = Constant() {r=1}
  n27 = Constant()
  n28 = QualitySwitch(Default=@n26, Low=@n27, High=@n26, Medium=@n27, Epic=@n26)
  SetMaterialAttributes!(
   MaterialAttributes=@n24.MaterialAttributes
   Base Color=Multiply(A=LinearInterpolate(A=@n25, Alpha=@n28) {constA=1}, B=@n24.Base Color)
   Roughness=Constant()
   Specular=Constant() {r=1}
   Normal
   Tangent
   Emissive Color=Add(A=@n24.Emissive Color, B=NamedRerouteUsage())
   Opacity=NamedRerouteUsage()
  ) {attributeSetTypes [
    #1
    #4
    #3
    #5
    #9
    #6
    #7
   ]}
 ]
}
